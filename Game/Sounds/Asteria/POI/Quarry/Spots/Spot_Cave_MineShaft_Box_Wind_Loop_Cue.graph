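# Looping ambient spot sound cue for box-shaped mine shaft cave wind in the Asteria Quarry POI, reusing the Winterfest exterior wind loop.
# The looping wave feeds a Modulator, then two nested, endlessly looping Envelopers.
# The inner one sweeps pitch and the outer gently dips volume on a different period, so the loop keeps drifting.
# Overrides add box attenuation with listener focus and a tight concurrency limit, on the ambient sound class.
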SoundCue /Game/Sounds/Asteria/POI/Quarry/Spots/Spot_Cave_MineShaft_Box_Wind_Loop_Cue
root=Enveloper_3
nodes [
 {
  id=Enveloper_1
  children [Modulator_0]
  properties {
   bLoop=true
   bLoopIndefinitely=true
   loopEnd=30
   pitchCurve {editorCurveData {keys [{arriveTangent=-0.0033, interpMode=RCIM_Cubic, time=0}, <at 15 0.8 RCIM_Cubic>, {interpMode=RCIM_Cubic, leaveTangent=0.0033, time=30}]}}
   volumeCurve {editorCurveData {keys [<at 0 1>]}}
  }
 }
 {
  id=Enveloper_3
  children [Enveloper_1]
  properties {
   bLoop=true
   bLoopIndefinitely=true
   loopEnd=20
   pitchCurve {editorCurveData {keys [{time=0}]}}
   volumeCurve {editorCurveData {keys [
      {arriveTangent=-0.01, interpMode=RCIM_Cubic, leaveTangent=-0.01, time=0, value=1}
      <at 10 0.9 RCIM_Cubic>
      {arriveTangent=0.01, interpMode=RCIM_Cubic, leaveTangent=0.01, time=20, value=1}
     ]}}
  }
 }
 {id=Modulator_0, children [WavePlayer_0]}
 {id=WavePlayer_0, properties {bLooping=true, soundWaveAssetPtr=/Game/Athena/Sounds/Events/Winterfest/Winterfest_Exterior_Wind_Loop}}
]
properties {
 bOverrideAttenuation=true
 bOverrideConcurrency=true
 maxDistance=600
 soundClassObject=&Fort_Ambient_Athena
 volumeMultiplier=0.376
 attenuationOverrides {
  attenuationShape=Box
  bEnableListenerFocus=true
  falloffDistance=350
  focusAzimuth=15
  focusVolumeAttenuation=1.2
  nonFocusAzimuth=35
  nonFocusVolumeAttenuation=0.4
  attenuationShapeExtents=<200 200 250>
 }
 concurrencyOverrides {maxCount=2, voiceStealReleaseTime=2}
}
