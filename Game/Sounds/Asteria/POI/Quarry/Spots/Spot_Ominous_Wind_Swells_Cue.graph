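# Sound cue for ominous wind swell one-shot spot ambience at the Asteria Quarry point of interest.
# A Random node with equal weights picks one of seven cave wind swell waves, reused from the Artemis Hollow Base sound set.
# It plays through a tile ambient one-shot attenuation and an outdoor ambient sound class, with bus and submix sends disabled.
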SoundCue /Game/Sounds/Asteria/POI/Quarry/Spots/Spot_Ominous_Wind_Swells_Cue
root=Random_0
nodes [
 {
  id=Random_0
  children [WavePlayer_7, WavePlayer_8, WavePlayer_9, WavePlayer_10, WavePlayer_11, WavePlayer_12, WavePlayer_13]
  properties {weights [1,1,1,1,1,1,1]}
 }
 {id=WavePlayer_10, properties {soundWaveAssetPtr=/Game/Sounds/Artemis/POI/Hollow_Base/Spots/Spot_Cave_Ominous_Wind_Swells_03}}
 {id=WavePlayer_11, properties {soundWaveAssetPtr=/Game/Sounds/Artemis/POI/Hollow_Base/Spots/Spot_Cave_Ominous_Wind_Swells_04}}
 {id=WavePlayer_12, properties {soundWaveAssetPtr=/Game/Sounds/Artemis/POI/Hollow_Base/Spots/Spot_Cave_Ominous_Wind_Swells_05}}
 {id=WavePlayer_13, properties {soundWaveAssetPtr=/Game/Sounds/Artemis/POI/Hollow_Base/Spots/Spot_Cave_Ominous_Wind_Swells_06}}
 {id=WavePlayer_7, properties {soundWaveAssetPtr=/Game/Sounds/Artemis/POI/Hollow_Base/Spots/Spot_Cave_Ominous_Wind_Swells_07}}
 {id=WavePlayer_8, properties {soundWaveAssetPtr=/Game/Sounds/Artemis/POI/Hollow_Base/Spots/Spot_Cave_Ominous_Wind_Swells_01}}
 {id=WavePlayer_9, properties {soundWaveAssetPtr=/Game/Sounds/Artemis/POI/Hollow_Base/Spots/Spot_Cave_Ominous_Wind_Swells_02}}
]
properties {
 attenuationSettings=/Game/Sounds/Attenuation/Fort_TileAmbient_OneShot_Att
 bEnableBusSends=false
 bEnableSubmixSends=false
 maxDistance=8000
 soundClassObject=&Fort_Ambient_Outdoors_Athena
 volumeMultiplier=1.002
}
