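# Sound cue for big rock fall spot one-shots in the Asteria Quarry point of interest, classed as outdoor ambient one-shots.
# A random node picks one of eight rock fall waves with equal weight.
# A modulator at the root varies the pitch of each play around normal for extra variety.
# The cue plays at reduced volume and references a shared wide ambient one-shot attenuation asset.
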SoundCue /Game/Sounds/Asteria/POI/Quarry/Spots/Spot_Rock_Big_Fall_Cue
root=Modulator_0
nodes [
 {id=Modulator_0, children [Random_1], properties {pitchMax=1.2, pitchMin=0.8}}
 {
  id=Random_1
  children [WavePlayer_5, WavePlayer_6, WavePlayer_7, WavePlayer_8, WavePlayer_9, WavePlayer_10, WavePlayer_24, WavePlayer_25]
  properties {weights [1,1,1,1,1,1,1,1]}
 }
 {id=WavePlayer_10, properties {soundWaveAssetPtr=&/Spot_Rock_Big_Fall_5}}
 {id=WavePlayer_24, properties {soundWaveAssetPtr=&/Spot_Rock_Big_Fall_6}}
 {id=WavePlayer_25, properties {soundWaveAssetPtr=&/Spot_Rock_Big_Fall_7}}
 {id=WavePlayer_5, properties {soundWaveAssetPtr=&/Spot_Rock_Big_Fall_8}}
 {id=WavePlayer_6, properties {soundWaveAssetPtr=&/Spot_Rock_Big_Fall_1}}
 {id=WavePlayer_7, properties {soundWaveAssetPtr=&/Spot_Rock_Big_Fall_2}}
 {id=WavePlayer_8, properties {soundWaveAssetPtr=&/Spot_Rock_Big_Fall_3}}
 {id=WavePlayer_9, properties {soundWaveAssetPtr=&/Spot_Rock_Big_Fall_4}}
]
properties {
 attenuationSettings=/Game/Sounds/Attenuation/Fort_TileAmbient_OneShot_Wider_Att
 maxDistance=8000
 soundClassObject=&Fort_Ambient_OneShots_Outdoors
 volumeMultiplier=0.251
 attenuationOverrides {bAttenuate=false}
}
