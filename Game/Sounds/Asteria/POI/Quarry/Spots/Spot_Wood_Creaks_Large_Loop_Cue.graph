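# Looping ambient spot cue that plays intermittent large wood creaks, placed by path in the Asteria Quarry POI.
# A loop wraps a random delay of a few seconds, then a quality switch into pitch and volume modulators over random creak waves.
# The first quality slot draws from six waves, while the other three share a three-wave branch.
# It overrides attenuation and limits concurrency, and uses the Fort_Doors sound class.
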SoundCue /Game/Sounds/Asteria/POI/Quarry/Spots/Spot_Wood_Creaks_Large_Loop_Cue
root=Looping_0
nodes [
 {id=Delay_0, children [QualityLevel_0], properties {delayMax=8, delayMin=3}}
 {id=Looping_0, children [Delay_0]}
 {id=Modulator_0, children [Random_0], properties {pitchMin=0.75, volumeMin=0.5}}
 {id=Modulator_36, children [Random_4], properties {pitchMax=1, pitchMin=0.8, volumeMax=1, volumeMin=0.5}}
 {id=QualityLevel_0, children [Modulator_36,Modulator_0,Modulator_0,Modulator_0]}
 {id=Random_0, children [WavePlayer_8,WavePlayer_9,WavePlayer_7], properties {weights [1,1,1]}}
 {
  id=Random_4
  children [WavePlayer_4, WavePlayer_5, WavePlayer_6, WavePlayer_14, WavePlayer_3, WavePlayer_16]
  properties {weights [1,1,1,1,1,1]}
 }
 {id=WavePlayer_14, properties {soundWaveAssetPtr=/Game/Sounds/Trans_AmbientNonLoop/Wood/CreakHouseB002}}
 {id=WavePlayer_16, properties {soundWaveAssetPtr=/Game/Sounds/Trans_AmbientNonLoop/Wood/CreakHouseB003}}
 {id=WavePlayer_3, properties {soundWaveAssetPtr=/Game/Sounds/Trans_AmbientNonLoop/Wood/CreakHouseB001}}
 {id=WavePlayer_4, properties {soundWaveAssetPtr=/Game/Sounds/Trans_AmbientNonLoop/Wood/CreakHouseA001}}
 {id=WavePlayer_5, properties {soundWaveAssetPtr=/Game/Sounds/Trans_AmbientNonLoop/Wood/CreakHouseA002}}
 {id=WavePlayer_6, properties {soundWaveAssetPtr=/Game/Sounds/Trans_AmbientNonLoop/Wood/CreakHouseA003}}
 {id=WavePlayer_7, properties {soundWaveAssetPtr=&CreakHouseB001}}
 {id=WavePlayer_8, properties {soundWaveAssetPtr=&CreakHouseA001}}
 {id=WavePlayer_9, properties {soundWaveAssetPtr=&CreakHouseA002}}
]
properties {
 bOverrideAttenuation=true
 bOverrideConcurrency=true
 maxDistance=1350
 soundClassObject=&Fort_Doors
 volumeMultiplier=1
 attenuationOverrides {falloffDistance=1000, reverbWetLevelMin=1, attenuationShapeExtents {x=350}}
 concurrencyOverrides {maxCount=2}
 soundSubmixSends [{minSendLevel=0, sendLevel=0, soundSubmix=null}]
}
